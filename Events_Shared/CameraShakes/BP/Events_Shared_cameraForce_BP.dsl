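; Shared event actor that triggers a configurable camera shake and controller force feedback, either radius-based at its location or directly on the local player.
; The radius events use a small/medium/large size setting to pick force feedback attenuation assets or set shake source radii, with a fixed outer falloff margin.
; Lobby events bypass attenuation, targeting player 0's Fortnite controller and camera manager.
; The construction script previews attenuation and optional debug spheres in editor.
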(event Custom|forceFeedbackRadius
  (bind _defaultsceneroot (Variables|Default|GetDefaultSceneRoot))
  (bind _returnvalue (Transformation|GetWorldLocation _defaultsceneroot))
  (bind _forcefeedback (Variables|Default|GetForceFeedback))
  (bind _sizeofradius (Variables|Default|GetsizeOfRadius))
  ; force feedback attenuation and effect
  (if (== _sizeofradius 0)
    (Variables|/Script/Engine.ForceFeedbackComponent|SetAttenuationSettings "/Events_Shared/ForceFeedback/Events_Shared_00small_FA.Events_Shared_00small_FA" _forcefeedback)
    (ForceFeedback|SetForceFeedbackEffect _forcefeedback (Variables|Default|GetforceEffect))
    (ForceFeedback|Play _forcefeedback 0.0)
    (if (Variables|Default|GetpieDebugSphere)
      (Rendering|Debug|DrawDebugSphere _returnvalue (Variables|Default|GetsmallRadius) 11 "(R=0.000000,G=0.258183,B=1.000000,A=1.000000)" 4.0))
    (elif (== _sizeofradius 1)
      (Variables|/Script/Engine.ForceFeedbackComponent|SetAttenuationSettings "/Events_Shared/ForceFeedback/Events_Shared_01medium_FA.Events_Shared_01medium_FA" _forcefeedback)
      (ForceFeedback|SetForceFeedbackEffect _forcefeedback (Variables|Default|GetforceEffect))
      (ForceFeedback|Play _forcefeedback 0.0)
      (if (Variables|Default|GetpieDebugSphere)
        (Rendering|Debug|DrawDebugSphere _returnvalue (Variables|Default|GetmediumRadius) 11 "(R=0.000000,G=0.258183,B=1.000000,A=1.000000)" 4.0))
      (elif (== 2 _sizeofradius)
        (Variables|/Script/Engine.ForceFeedbackComponent|SetAttenuationSettings "/Events_Shared/ForceFeedback/Events_Shared_02large_FA.Events_Shared_02large_FA" _forcefeedback)
        (ForceFeedback|SetForceFeedbackEffect _forcefeedback (Variables|Default|GetforceEffect))
        (ForceFeedback|Play _forcefeedback 0.0)
        (if (Variables|Default|GetpieDebugSphere)
          (Rendering|Debug|DrawDebugSphere _returnvalue (Variables|Default|GetlargeRadius) 11 "(R=0.000000,G=0.258183,B=1.000000,A=1.000000)" 4.0))))))

(event Custom|cameraShakeRadius
  (bind _defaultsceneroot (Variables|Default|GetDefaultSceneRoot))
  (bind _returnvalue (Transformation|GetWorldLocation _defaultsceneroot))
  (bind _camerashakesource (Variables|Default|GetCameraShakeSource))
  (bind _camerashakesource_2 (Variables|Default|GetCameraShakeSource))
  (bind _smallradius (Variables|Default|GetsmallRadius))
  (bind _camerashakesource_3 (Variables|Default|GetCameraShakeSource))
  (bind _sizeofradius (Variables|Default|GetsizeOfRadius))
  (bind _largeradius (Variables|Default|GetlargeRadius))
  (bind _mediumradius (Variables|Default|GetmediumRadius))
  ; camera shake attenuation and bp shake
  (if (== _sizeofradius 0)
    (Variables|/Script/Engine.CameraShakeSourceComponent|SetInnerAttenuationRadius _smallradius _camerashakesource_2)
    (Variables|/Script/Engine.CameraShakeSourceComponent|SetOuterAttenuationRadius (+ _smallradius 5e+03) _camerashakesource_2)
    (CameraShake|StartCameraShake :self _camerashakesource_2 :InCameraShake (Variables|Default|GetcamShakeBase) :Scale 1.0 :PlaySpace "CameraLocal")
    (if (Variables|Default|GetpieDebugSphere)
      (Rendering|Debug|DrawDebugSphere _returnvalue (Variables|Default|GetsmallRadius) 11 "(R=0.000000,G=0.258183,B=1.000000,A=1.000000)" 4.0))
    (elif (== _sizeofradius 1)
      (Variables|/Script/Engine.CameraShakeSourceComponent|SetInnerAttenuationRadius _mediumradius _camerashakesource_3)
      (Variables|/Script/Engine.CameraShakeSourceComponent|SetOuterAttenuationRadius (+ _mediumradius 5e+03) _camerashakesource_3)
      (CameraShake|StartCameraShake :self _camerashakesource_3 :InCameraShake (Variables|Default|GetcamShakeBase) :Scale 1.0 :PlaySpace "CameraLocal")
      (if (Variables|Default|GetpieDebugSphere)
        (Rendering|Debug|DrawDebugSphere _returnvalue (Variables|Default|GetmediumRadius) 11 "(R=0.000000,G=0.258183,B=1.000000,A=1.000000)" 4.0))
      (elif (== 2 _sizeofradius)
        (Variables|/Script/Engine.CameraShakeSourceComponent|SetInnerAttenuationRadius _largeradius _camerashakesource)
        (Variables|/Script/Engine.CameraShakeSourceComponent|SetOuterAttenuationRadius (+ _largeradius 5e+03) _camerashakesource)
        (CameraShake|StartCameraShake :self _camerashakesource :InCameraShake (Variables|Default|GetcamShakeBase) :Scale 1.0 :PlaySpace "CameraLocal")
        (if (Variables|Default|GetpieDebugSphere)
          (Rendering|Debug|DrawDebugSphere _returnvalue (Variables|Default|GetlargeRadius) 11 "(R=0.000000,G=0.258183,B=1.000000,A=1.000000)" 4.0))))))

(event Custom|cameraShakeLobby
  ; camera shake lobby
  (CameraShakes|StartCameraShakefromSource :self (Variables|/Script/Engine.PlayerController|GetPlayerCameraManager (Fortnite|GetFortPlayerControllerfromActor (Game|GetPlayerPawn 0))) :ShakeClass (Variables|Default|GetcamShakeBase) :Scale 1.0 :PlaySpace "CameraLocal"))

(event Custom|forceFeedbackLobby
  ; force feedback lobby
  (Game|Feedback|ClientPlayForceFeedback (Fortnite|GetFortPlayerControllerfromActor (Game|GetPlayerPawn 0)) (Variables|Default|GetforceEffect) "None" false false false))

(fn ConstructionScript
  (bind _defaultsceneroot (Variables|Default|GetDefaultSceneRoot))
  (bind _returnvalue (Transformation|GetWorldLocation _defaultsceneroot))
  (bind _forcefeedback (Variables|Default|GetForceFeedback))
  (bind _sizeofradius (Variables|Default|GetsizeOfRadius))
  (if (== _sizeofradius 0)
    (Variables|/Script/Engine.ForceFeedbackComponent|SetAttenuationSettings "/Events_Shared/ForceFeedback/Events_Shared_00small_FA.Events_Shared_00small_FA" _forcefeedback)
    (if (Variables|Default|GeteditorDebugSphere)
      (Rendering|Debug|DrawDebugSphere _returnvalue (Variables|Default|GetsmallRadius) 11 "(R=0.000000,G=0.258183,B=1.000000,A=1.000000)" 2.0))
    (elif (== _sizeofradius 1)
      (Variables|/Script/Engine.ForceFeedbackComponent|SetAttenuationSettings "/Events_Shared/ForceFeedback/Events_Shared_01medium_FA.Events_Shared_01medium_FA" _forcefeedback)
      (if (Variables|Default|GeteditorDebugSphere)
        (Rendering|Debug|DrawDebugSphere _returnvalue (Variables|Default|GetmediumRadius) 11 "(R=0.000000,G=0.258183,B=1.000000,A=1.000000)" 2.0))
      (elif (== 2 _sizeofradius)
        (Variables|/Script/Engine.ForceFeedbackComponent|SetAttenuationSettings "/Events_Shared/ForceFeedback/Events_Shared_02large_FA.Events_Shared_02large_FA" _forcefeedback)
        (if (Variables|Default|GeteditorDebugSphere)
          (Rendering|Debug|DrawDebugSphere _returnvalue (Variables|Default|GetlargeRadius) 11 "(R=0.000000,G=0.258183,B=1.000000,A=1.000000)" 2.0))))))
